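# Niagara system for the portal effect of the ProngCling player emote, built from sprites facing a fixed Y axis.
# Ripple emits short-lived, growing sprites after a delay, animating their look through curve-driven dynamic material parameters.
# SideGlow holds a cyan glow that uses camera-to-facing dot products to narrow its width and raise its alpha when viewed edge-on.
# SpawnStar shows no modules or renderers, and neither front-end user bool is referenced by the modules shown.
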NiagaraSystem /BRCosmetics/Animation/Game/MainPlayer/Emotes/ProngCling/FX/NS_Emote_ProngCling_Portal
dependencies {
 usedDataInterfaces [&NiagaraDataInterfaceCurve, &NiagaraDataInterfaceCamera]
 usedDynamicInputs []
 usedModules [&SystemState]
}
emitters {
 Ripple {
  emitterUpdateScript [
   &EmitterState(Life Cycle Mode=::NewEnumerator1, Loop Behavior=::NewEnumerator1, Loop Delay=1, Loop Duration=0.6, UseLoopDelay=true)
   &SpawnRate(SpawnRate=3, Use Spawn Probability=false)
   &SpawnBurst_Instantaneous(Use Spawn Probability=false, UseLoopCountLimit=false)
  ]
  particleSpawnScript [
   &InitializeParticle(
    Color Mode=::NewEnumerator0
    Lifetime=0.8
    Position Offset=<5 2 -2>
    Sprite Size Mode=::NewEnumerator3
    Uniform Sprite Size=30
    UsePositionOffset=true
    Write Material Random=true
   )
   &SpriteFacingAndAlignment(Sprite Facing=<0 1 0>, WriteAlignment=false, WriteFacing=true)
  ]
  particleUpdateScript [
   &ParticleState()
   &DynamicMaterialParameters(
    Index 0 Param 1=&FloatFromCurve(
     FloatCurve=DataInterface() {propertyValues {Curve {keys [{arriveTangent=1.3374, interpMode=RCIM_Cubic, leaveTangent=1.3374, tangentMode=RCTM_User, time=0, value=1}]}}}
     CurveIndex=$Particles.Age
    )
    Index 0 Param 2=0.15
    Index 0 Param 3=&FloatFromCurve(
     FloatCurve=DataInterface() {propertyValues {Curve {keys [
         {arriveTangent=4.9025, interpMode=RCIM_Cubic, leaveTangent=4.9025, tangentMode=RCTM_User, time=0, value=0}
         {arriveTangent=4.4266, interpMode=RCIM_Cubic, tangentMode=RCTM_Break, time=0.5796, value=2}
        ]}}}
     CurveIndex=$Particles.Age
     Scale Curve=0.05
    )
    Index 0 Param 4=&FloatFromCurve(FloatCurve=DataInterface() {propertyValues {Curve {keys [
         {arriveTangent=-3.3445, interpMode=RCIM_Cubic, leaveTangent=-3.3445, tangentMode=RCTM_User, time=0, value=0.462}
         {arriveTangent=-1.1268, interpMode=RCIM_Cubic, leaveTangent=-1.2066, tangentMode=RCTM_Break, time=0.2509, value=-0.3002}
         {arriveTangent=-1.4939, interpMode=RCIM_Cubic, tangentMode=RCTM_Break, time=1, value=-1}
        ]}}})
    Index 1 Param 1=0.05
    Index 1 Param 2=0.5
    Index 1 Param 4=3
    Module.Index0Float0Write=true
    Module.Index0Float1Write=true
    Module.Index0Float2Write=true
    Module.Index0Float3Write=true
    Module.Index1Float0Write=true
    Module.Index1Float1Write=true
    Module.Index1Float2Write=true
    Module.Index1Float3Write=true
    Write Parameter Index 1=::NewEnumerator1
   )
   &ScaleSpriteSize(Uniform Curve Sprite Scale=DataInterface() {propertyValues {Curve {keys [
        {arriveTangent=4.1785, interpMode=RCIM_Cubic, leaveTangent=4.1785, tangentMode=RCTM_User, time=0, value=0}
        {arriveTangent=1.3778, interpMode=RCIM_Cubic, leaveTangent=1.3778, tangentMode=RCTM_User, time=0.1452, value=0.6}
        {arriveTangent=-0.0405, interpMode=RCIM_Cubic, leaveTangent=-0.0405, tangentMode=RCTM_User, time=1, value=1}
       ]}}})
   &SolveForcesAndVelocity(Clamp Velocity=false, Limit Acceleration=false)
  ]
  rendererClasses [&NiagaraSpriteRendererProperties]
  renderers [{rendererClass=&NiagaraSpriteRendererProperties}]
 }
 SideGlow {
  emitterUpdateScript [
   &EmitterState(Life Cycle Mode=::NewEnumerator1, Loop Behavior=::NewEnumerator1, Loop Duration=2.193, UseLoopDelay=false)
   &SpawnBurst_Instantaneous(Use Spawn Probability=false, UseLoopCountLimit=false)
  ]
  particleSpawnScript [
   &InitializeParticle(
    Color=<rgba 0 0.7287 1 0.5>
    Lifetime=$Emitter.CurrentLoopDuration
    Sprite Size Mode=::NewEnumerator3
    Uniform Sprite Size=90
    UsePositionOffset=false
    Write Material Random=true
   )
   &SpriteFacingAndAlignment(Sprite Alignment=<0 0 1>, Sprite Facing=<0 1 0>, WriteAlignment=true, WriteFacing=true)
   &CameraOffset(Camera Offset Amount=-10)
  ]
  particleUpdateScript [
   &ParticleState()
   &ScaleSpriteSize(
    Scale Factor=&Multiply_Vector2D(
     A=&Vector2DFromFloat(Value=&Multiply_Float(
       A=&FloatFromCurve(FloatCurve=DataInterface() {propertyValues {Curve {
           defaultValue=1
           keys [
            {arriveTangent=0.7744, interpMode=RCIM_Cubic, leaveTangent=0.7744, tangentMode=RCTM_User, time=0, value=1.018}
            {arriveTangentWeight=0.4212, interpMode=RCIM_Cubic, tangentMode=RCTM_User, time=0.2538, value=1.4541}
           ]
          }}})
       B=1
      ))
     B=&MakeVector2D(
      X=&RemapRange(
       Input Value To Remap=&Abs_Float(Float:NiagaraFloat=&DotProduct(
         Vector A=&ReturnCameraVector(
          Camera Vectors=::NewEnumerator4
          Coordinate Space=::Simulation
          Origin Position=$Particles.Position
          CameraQuery=DataInterface() {propertyValues {bRequireCurrentFrameData=false}}
         )
         Vector B=$Particles.SpriteFacing
        ))
       Input Min=0.2
       Input Max=0.9
       Output Min=0.13
       Output Max=1
      )
      Y=1
     )
    )
    Scale Sprite Size Mode=::NewEnumerator2
   ) {module=ScaleSpriteSize001}
   &ScaleColor(
    Scale Alpha=&FloatFromCurve(
     FloatCurve=DataInterface() {propertyValues {Curve {
        defaultValue=1
        keys [
         {interpMode=RCIM_Cubic, tangentMode=RCTM_User, time=0, value=0}
         {arriveTangentWeight=0.071, interpMode=RCIM_Cubic, leaveTangentWeight=0.23, tangentMode=RCTM_User, time=0.202, value=0.98}
         {arriveTangentWeight=0.2481, interpMode=RCIM_Cubic, leaveTangentWeight=0.0544, tangentMode=RCTM_User, time=0.8231, value=1}
         {tangentMode=RCTM_User, time=1, value=0}
        ]
       }}}
     Scale Curve=&Multiply_Float(
      A=2
      B=&FloatFromCurve(
       FloatCurve=DataInterface() {propertyValues {Curve {
          defaultValue=1
          keys [
           {arriveTangent=-2.6115, interpMode=RCIM_Cubic, leaveTangent=-2.6115, tangentMode=RCTM_User, time=0, value=1}
           {interpMode=RCIM_Cubic, tangentMode=RCTM_User, time=0.8007, value=0.4}
          ]
         }}}
       CurveIndex=&Abs_Float(Float:NiagaraFloat=&DotProduct(
         Vector A=&ReturnCameraVector(
          Camera Vectors=::NewEnumerator4
          Coordinate Space=::Simulation
          Origin Position=$Particles.Position
          CameraQuery=DataInterface() {propertyValues {bRequireCurrentFrameData=false}}
         )
         Vector B=$Particles.SpriteFacing
        ))
      )
     )
    )
    ScaleA=true
    ScaleRGB=false
   )
  ]
  rendererClasses [&NiagaraSpriteRendererProperties]
  renderers [{rendererClass=&NiagaraSpriteRendererProperties}]
 }
 SpawnStar {rendererClasses [], renderers []}
}
summary {
 systemName=NS_Emote_ProngCling_Portal
 emitters [
  {emitterName=SideGlow, rendererClasses [&NiagaraSpriteRendererProperties]}
  {emitterName=Ripple, rendererClasses [&NiagaraSpriteRendererProperties]}
  {emitterName=SpawnStar, rendererClasses []}
 ]
 userVariables [
  User.bIsFrontEnd {type=NiagaraBool}
  User.bIsFrontEndPreview {type=NiagaraBool}
 ]
}
