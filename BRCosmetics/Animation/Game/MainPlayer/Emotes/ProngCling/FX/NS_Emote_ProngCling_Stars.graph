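# Niagara system for the ProngCling player emote that throws a short, rainbow-tinted burst of stars, flare and sparks.
# Six sprite emitters anchor to an offset point from a shared skeletal mesh reader: a layered flare, spinning stars, motes, flickering fountain sparkles and two flash-streak layers flying outward.
# Spawn groups drive masks and visibility tags to vary stars and motes within one emitter, and a front-end modifier rescales the flare.
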NiagaraSystem /BRCosmetics/Animation/Game/MainPlayer/Emotes/ProngCling/FX/NS_Emote_ProngCling_Stars
dependencies {
 usedDataInterfaces [
  &NiagaraDataInterfaceColorCurve
  &NiagaraDataInterfaceCurve
  &NiagaraDataInterfaceArrayColor
  &NiagaraDataInterfaceSpriteRendererInfo
  &NiagaraDataInterfaceVector2DCurve
  &NiagaraDataInterfaceSkeletalMesh
 ]
 usedDynamicInputs []
 usedModules [
  &self:SystemSpawnScript.SystemScriptSource.SystemScriptGraph.NiagaraNodeAssignment_0.SetVariables_#1
  &SystemState
 ]
}
emitters {
 Flare_With_Streak {
  emitterUpdateScript [
   &EmitterState(Life Cycle Mode=::NewEnumerator1, Loop Behavior=::NewEnumerator1, Loop Duration=0.25, UseLoopDelay=false)
   &SpawnBurst_Instantaneous(Spawn Count=1, Spawn Time=0.02, Use Spawn Probability=false)
  ]
  particleSpawnScript [
   &InitializeParticle(
    Color Mode=::NewEnumerator0
    Lifetime=0.1
    Sprite Rotation Mode=::NewEnumerator2
    Sprite Size Mode=::NewEnumerator3
    Uniform Sprite Size=100
    UsePositionOffset=false
    Write Material Random=true
   )
   &DynamicMaterialParameters(
    Index 0 Param 1=&RandomRangeFloat(Minimum=0.4, Maximum=0.6)
    Index 0 Param 2=1
    Index 0 Param 3=1
    Index 0 Param 4=1
    Module.Param0WriteEnabled=true
    Module.Param1WriteEnabled=true
    Module.Param2WriteEnabled=true
    Module.Param3WriteEnabled=true
   )
   &SkeletalMeshLocation(
    MaskSpawnGroup=false
    Mesh Sampling Type=::NewEnumerator5
    Sampled Position Offset=<0 7 3>
    Sampled Position Offset Coordinate Space:ENiagaraSkelMeshTransforms=::NewEnumerator2
    Skeletal Mesh=$System.Skeletal Mesh Reader
    Socket Sampling Mode=::NewEnumerator5
   )
   /SetVariables() {isSetParameters=true}
  ]
  particleUpdateScript [
   &ParticleState()
   &SolveForcesAndVelocity(Clamp Velocity=false, Limit Acceleration=false)
   &ScaleColor(
    Scale Alpha=&Multiply_Float(
     A=1
     B=&FloatFromCurve(FloatCurve=DataInterface() {propertyValues {Curve {keys [<at 0 0>,<at 0.2 1>,<at 0.8 1>,<at 1 0>]}}})
    )
    Scale RGB=&MakeVectorFromLinearColorRGB(Color=&ColorFromCurve(ColorCurve=DataInterface() {propertyValues {
        AlphaCurve {keys []}
        BlueCurve {keys [
          {arriveTangent=0.9988, interpMode=RCIM_Cubic, time=0.0338, value=0}
          <at 0.1524 0>
          <at 0.2832 0.5879>
          <at 0.3972 1.6553>
          <at 0.4931 2>
          {arriveTangent=-7.1183, interpMode=RCIM_Cubic, leaveTangent=-7.1183, time=0.6265, value=2}
          <at 0.7547 0.1373>
          <at 0.87 0>
          <at 0.9931 0>
         ]}
        GreenCurve {keys [
          {arriveTangent=0.6763, interpMode=RCIM_Cubic, time=0.0338, value=1.2308}
          <at 0.1524 0.6992>
          <at 0.2832 0.607>
          <at 0.3972 0.5347>
          <at 0.4931 0.5753>
          {arriveTangent=5.4447, interpMode=RCIM_Cubic, leaveTangent=5.4447, time=0.6265, value=1.175}
          <at 0.7547 2>
          <at 0.87 1.0573>
          <at 0.9931 0.5225>
         ]}
        RedCurve {keys [
          {arriveTangent=-1, interpMode=RCIM_Cubic, time=0.0338, value=2}
          <at 0.1524 2>
          <at 0.2832 2>
          <at 0.3972 2>
          <at 0.4931 0.8161>
          {arriveTangent=0.6002, interpMode=RCIM_Cubic, leaveTangent=0.6002, time=0.6265, value=0.5742}
          <at 0.7547 0.9731>
          <at 0.87 2>
          <at 0.9931 2>
         ]}
       }}))
    ScaleA=true
    ScaleRGB=true
   )
   &ScaleSpriteSize(Scale Factor=&Vector2DFromFloat(Value=&FloatFromCurve(
      FloatCurve=DataInterface() {propertyValues {Curve {keys [<at 0 1.5 RCIM_Cubic>,<at 1 0>]}}}
      Scale Curve=&FN_FrontEndFloatModifier(Gameplay=1.25, Front End=1)
     )))
   &CameraOffset(Camera Offset Amount=50)
   /SetVariables(Particles.Flash size:Vector2f=&Multiply_Vector2D_ByFloat(
     Float=&Multiply_Float(
      A=70
      B=&FloatFromCurve(
       FloatCurve=DataInterface() {propertyValues {Curve {keys [
           {arriveTangent=4.1785, interpMode=RCIM_Cubic, leaveTangent=4.1785, tangentMode=RCTM_User, time=0, value=0}
           {arriveTangent=0.9995, interpMode=RCIM_Cubic, leaveTangent=0.9995, tangentMode=RCTM_User, time=0.1378, value=0.6825}
           {arriveTangent=-0.0405, interpMode=RCIM_Cubic, leaveTangent=-0.0405, tangentMode=RCTM_User, time=0.62, value=1}
          ]}}}
       Scale Curve=&FN_FrontEndFloatModifier(Gameplay=1.3, Front End=1.1)
      )
     )
     Vector2D=<1.5 1>
    )) {isSetParameters=true}
   /SetVariables(Particles.Flash color:LinearColor=&MakeLinearColorFromVectorAndFloat(
     "Vector (RGB)"=&MakeVectorFromLinearColorRGB(Color=&ColorFromCurve(ColorCurve=DataInterface() {propertyValues {
         AlphaCurve {keys []}
         BlueCurve {keys [
           {arriveTangent=0.9988, interpMode=RCIM_Cubic, time=0.1157, value=0}
           <at 0.2225 0>
           <at 0.3976 0.2939>
           <at 0.561 0.8276>
           <at 0.7351 0.1025>
           <at 1 0 RCIM_Cubic>
          ]}
         GreenCurve {keys [
           {arriveTangent=0.6763, interpMode=RCIM_Cubic, time=0.1157, value=0.3556}
           <at 0.2225 0.3496>
           <at 0.3976 0.3035>
           <at 0.561 0.2673>
           <at 0.7351 0.3444>
           <at 1 0.3006 RCIM_Cubic>
          ]}
         RedCurve {keys [
           {arriveTangent=-1, interpMode=RCIM_Cubic, time=0.1157, value=1}
           <at 0.2225 1>
           <at 0.3976 1>
           <at 0.561 1>
           <at 0.7351 1>
           <at 1 1 RCIM_Cubic>
          ]}
        }}))
     "Float (Alpha)"=&Multiply_Float(
      A=&Multiply_Float(
       A=1
       B=&FloatFromCurve(FloatCurve=DataInterface() {propertyValues {Curve {keys [<at 0 0>,<at 0.124 1>,<at 0.496 1>,<at 0.62 0>]}}})
      )
      B=1
     )
    )) {isSetParameters=true}
   /SetVariables(
    Particles.Rainbow Color:LinearColor=&MakeLinearColorFromVectorAndFloat(
     "Vector (RGB)"=&MakeVectorFromLinearColorRGB(Color=&ColorFromCurve(ColorCurve=DataInterface() {propertyValues {
         AlphaCurve {keys []}
         BlueCurve {keys [<at 0 5>,<at 0.0525 2 RCIM_Cubic>,<at 0.173 1 RCIM_Cubic>]}
         GreenCurve {keys [<at 0 5>,<at 0.0525 2 RCIM_Cubic>,<at 0.173 1 RCIM_Cubic>]}
         RedCurve {keys [<at 0 5>,<at 0.0525 2 RCIM_Cubic>,<at 0.173 1 RCIM_Cubic>]}
        }}))
     "Float (Alpha)"=&FloatFromCurve(FloatCurve=DataInterface() {propertyValues {Curve {keys [<at 0 0>,<at 0.0571 1>,<at 0.442 1>,<at 0.62 0>]}}})
    )
    Particles.Rainbow Size:Vector2f=&Vector2DFromFloat(Value=&Multiply_Float(
      A=130
      B=&FloatFromCurve(FloatCurve=DataInterface() {propertyValues {Curve {keys [
           {arriveTangent=4.1785, interpMode=RCIM_Cubic, leaveTangent=4.1785, tangentMode=RCTM_User, time=0, value=0}
           {arriveTangent=0.9995, interpMode=RCIM_Cubic, leaveTangent=0.9995, tangentMode=RCTM_User, time=0.1378, value=0.6825}
           {arriveTangent=-0.0405, interpMode=RCIM_Cubic, leaveTangent=-0.0405, tangentMode=RCTM_User, time=0.62, value=1}
          ]}}})
     ))
   ) {isSetParameters=true}
   /SetVariables(Particles.Flash Alt Color:LinearColor=&MakeLinearColorFromVectorAndFloat(
     "Vector (RGB)"=&MakeVectorFromLinearColorRGB(Color=&LerpLinearColors(
       StartColor=&ColorFromCurve(ColorCurve=DataInterface() {propertyValues {
          AlphaCurve {keys []}
          BlueCurve {keys [
            <at 0.0218 0.0009>
            <at 0.1275 0>
            {arriveTangent=1.3513, interpMode=RCIM_Cubic, leaveTangent=1.3513, time=0.2321, value=0}
            <at 0.3528 0.3045>
            <at 0.4243 1.2927>
            <at 0.5023 1.0775>
            <at 0.6005 1.6553>
            <at 0.6678 2>
            <at 0.7404 2>
            <at 0.7992 0.9996>
            <at 0.8974 0.4109>
            <at 0.954 0.1506>
            <at 0.9935 0.0037>
           ]}
          GreenCurve {keys [
            <at 0.0218 0.2418>
            <at 0.1275 0.6992>
            {arriveTangent=5.7731, interpMode=RCIM_Cubic, leaveTangent=5.7731, time=0.2321, value=1.2308}
            <at 0.3528 2>
            <at 0.4243 1.0136>
            <at 0.5023 0.4161>
            <at 0.6005 0.5347>
            <at 0.6678 0.5753>
            <at 0.7404 1.1435>
            <at 0.7992 2>
            <at 0.8974 2>
            <at 0.954 1.1731>
            <at 0.9935 0.5119>
           ]}
          RedCurve {keys [
            <at 0.0218 2>
            <at 0.1275 2>
            {arriveTangent=-8.8763, interpMode=RCIM_Cubic, leaveTangent=-8.8763, time=0.2321, value=2}
            <at 0.3528 0>
            <at 0.4243 0.4048>
            <at 0.5023 0.7679>
            <at 0.6005 2>
            <at 0.6678 0.8161>
            <at 0.7404 0.4289>
            <at 0.7992 0.3334>
            <at 0.8974 1.2627>
            <at 0.954 2>
            <at 0.9935 2>
           ]}
         }})
       EndColor=&ColorFromCurve(ColorCurve=DataInterface() {propertyValues {
          AlphaCurve {keys []}
          BlueCurve {keys [<at 0 0.2943>,<at 0.065 0.1506>,<at 0.1529 0>,<at 0.2561 0.1952>,<at 0.3245 0.8131>,<at 0.4012 1.2064>,<at 0.5281 1.3852>,<at 0.6383 0.1176>,<at 0.7415 1.1868>,<at 0.8503 1.4314>,<at 0.927 0.7128>,<at 1 0.1633>]}
          GreenCurve {keys [<at 0 0.4148>,<at 0.065 1.1731>,<at 0.1529 0.7704>,<at 0.2561 0.2669>,<at 0.3245 0.1671>,<at 0.4012 0.4836>,<at 0.5281 2>,<at 0.6383 2>,<at 0.7415 0.9494>,<at 0.8503 0.2857>,<at 0.927 0.1733>,<at 1 0.2129>]}
          RedCurve {keys [<at 0 2>,<at 0.065 2>,<at 0.1529 1.6519>,<at 0.2561 1.377>,<at 0.3245 1.0914>,<at 0.4012 0.3865>,<at 0.5281 0.3295>,<at 0.6383 0.6298>,<at 0.7415 0.4047>,<at 0.8503 0.9434>,<at 0.927 2>,<at 1 2>]}
         }})
       LerpFactor=&UniformAOrBFloat(A=0, Evaluation Type=::NewEnumerator0, Recalculate Random Each Loop=true, Override Seed=false)
      ))
     "Float (Alpha)"=&Multiply_Float(
      A=1
      B=&FloatFromCurve(FloatCurve=DataInterface() {propertyValues {Curve {keys [<at 0 0>,<at 0.2 1>,<at 0.8 1>,<at 1 0>]}}})
     )
    )) {isSetParameters=true}
  ]
  rendererClasses [&NiagaraSpriteRendererProperties]
  renderers [
   {rendererClass=&NiagaraSpriteRendererProperties}
   {rendererClass=&NiagaraSpriteRendererProperties, rendererIndex=1}
   {rendererClass=&NiagaraSpriteRendererProperties, rendererIndex=2}
   {rendererClass=&NiagaraSpriteRendererProperties, rendererIndex=3}
  ]
 }
 Fountain {
  emitterUpdateScript [
   &EmitterState(Life Cycle Mode=::NewEnumerator1, Loop Behavior=::NewEnumerator1, Loop Duration=1, UseLoopDelay=false)
   &SpawnBurst_Instantaneous(
    Spawn Count=&RandomRangeInt(Minimum=7, Maximum=11)
    Spawn Time=&RandomRangeFloat(Minimum=0.055, Maximum=0.155)
    Use Spawn Probability=false
    UseLoopCountLimit=false
   )
  ]
  particleSpawnScript [
   &InitializeParticle(
    Color=&SelectLinearColorFromArray(
     Array Sampling Mode=::NewEnumerator0
     Direct Array Index=?
     Color Selection Array=DataInterface() {propertyValues {ColorData [<rgba 1 0.6499 0 1>, <rgba 1 0.0529 0 1>, <rgba 0.2767 0 1 1>, <rgba 0 0.5017 1 1>, <rgba 0.0024 1 0.2401 1>]}}
     Random Seed=0
    )
    Lifetime Max=0.6
    Lifetime Min=0.18
    Lifetime Mode=::NewEnumerator1
    Mass Mode=::NewEnumerator3
    Sprite Rotation Mode=::NewEnumerator2
    Sprite Size Mode=::NewEnumerator1
    Uniform Sprite Size Max=15
    Uniform Sprite Size Min=3
    UsePositionOffset=false
    Write Material Random=true
   )
   &SkeletalMeshLocation(
    MaskSpawnGroup=false
    Mesh Sampling Type=::NewEnumerator5
    Sampled Position Offset=<0 7 3>
    Sampled Position Offset Coordinate Space:ENiagaraSkelMeshTransforms=::NewEnumerator2
    Skeletal Mesh=$System.Skeletal Mesh Reader
    Socket Sampling Mode=::NewEnumerator5
   )
   &ShapeLocation(MaskSpawnGroup=false, Override Seed=false, Sphere Radius=40, UseOwnerScale=true)
   &AddVelocity(
    Module.Use Falloff Distance=false
    Origin Offset:NiagaraPosition=$Particles.SkeletalMeshLocation.SampledPosition
    Override Seed=false
    Velocity Mode=::NewEnumerator1
    Velocity Speed=&RandomRangeFloat(Minimum=100, Maximum=200)
   )
   &ApplyInitialForces()
  ]
  particleUpdateScript [
   &ParticleState()
   &ScaleSpriteSize(Uniform Curve Sprite Scale=DataInterface() {propertyValues {Curve {
       defaultValue=0
       keys [
        {interpMode=RCIM_Cubic, tangentMode=RCTM_User, time=0, value=1}
        {arriveTangent=-2.7658, interpMode=RCIM_Cubic, leaveTangent=-2.7658, tangentMode=RCTM_User, time=1, value=0}
       ]
      }}})
   &ScaleColor(
    Scale Alpha=&FloatFromCurve(
     FloatCurve=DataInterface() {propertyValues {Curve {keys [
         <at -0.0025 0 RCIM_Cubic>
         {arriveTangent=0.0019, interpMode=RCIM_Cubic, leaveTangent=0.0019, time=0.5456, value=1}
         <at 1.0011 0.002 RCIM_Cubic>
        ]}}}
     Scale Curve=&RandomRangeFloat(Minimum=0.35, Maximum=0.7)
    )
    Scale RGB=&MakeVectorFromLinearColorRGB(Color=&ColorFromCurve(ColorCurve=DataInterface() {propertyValues {
        AlphaCurve {keys []}
        BlueCurve {keys [<at 0 1 RCIM_Cubic>,<at 0.0837 1>,<at 0.1169 10>,<at 0.1475 1>,<at 0.3542 1>,<at 0.3848 10>,<at 0.4154 1>,<at 0.6004 1>,<at 0.6298 10>,<at 0.6604 1>,<at 0.7982 1>,<at 0.8262 10>,<at 0.8543 1>,<at 1 1 RCIM_Cubic>]}
        GreenCurve {keys [<at 0 1 RCIM_Cubic>,<at 0.0837 1>,<at 0.1169 10>,<at 0.1475 1>,<at 0.3542 1>,<at 0.3848 10>,<at 0.4154 1>,<at 0.6004 1>,<at 0.6298 10>,<at 0.6604 1>,<at 0.7982 1>,<at 0.8262 10>,<at 0.8543 1>,<at 1 1 RCIM_Cubic>]}
        RedCurve {keys [<at 0 1 RCIM_Cubic>,<at 0.0837 1>,<at 0.1169 10>,<at 0.1475 1>,<at 0.3542 1>,<at 0.3848 10>,<at 0.4154 1>,<at 0.6004 1>,<at 0.6298 10>,<at 0.6604 1>,<at 0.7982 1>,<at 0.8262 10>,<at 0.8543 1>,<at 1 1 RCIM_Cubic>]}
       }}))
    ScaleA=true
    ScaleRGB=true
   )
   &Drag(Drag=&RandomRangeFloat(Minimum=5, Maximum=10), Use Linear Drag=true, Use Rotational Drag=true)
   &SolveForcesAndVelocity(Clamp Velocity=false, Limit Acceleration=false)
   /SetVariables(
    Particles.Linear Color=&MultiplyLinearColor(
     Linear Color=$Particles.Color
     Scale Factor=&MakeVector4_FromLinearColor(LinearColor=&ColorFromCurve(ColorCurve=DataInterface() {propertyValues {
         AlphaCurve {keys [<at 0.1092 1 RCIM_Cubic>,<at 0.5328 0.4323>,<at 1 0 RCIM_Cubic>]}
         BlueCurve {keys [<at 0 1 RCIM_Cubic>,<at 0.0837 1>,<at 0.1169 5>,<at 0.1475 1>,<at 0.3542 1>,<at 0.3848 5>,<at 0.4154 1>,<at 0.6004 1>,<at 0.6298 5>,<at 0.6604 1>,<at 0.7982 1>,<at 0.8262 5>,<at 0.8543 1>,<at 1 1 RCIM_Cubic>]}
         GreenCurve {keys [<at 0 1 RCIM_Cubic>,<at 0.0837 1>,<at 0.1169 5>,<at 0.1475 1>,<at 0.3542 1>,<at 0.3848 5>,<at 0.4154 1>,<at 0.6004 1>,<at 0.6298 5>,<at 0.6604 1>,<at 0.7982 1>,<at 0.8262 5>,<at 0.8543 1>,<at 1 1 RCIM_Cubic>]}
         RedCurve {keys [<at 0 1 RCIM_Cubic>,<at 0.0837 1>,<at 0.1169 5>,<at 0.1475 1>,<at 0.3542 1>,<at 0.3848 5>,<at 0.4154 1>,<at 0.6004 1>,<at 0.6298 5>,<at 0.6604 1>,<at 0.7982 1>,<at 0.8262 5>,<at 0.8543 1>,<at 1 1 RCIM_Cubic>]}
        }}))
    )
    Particles.Vector 2D:Vector2f=&Multiply_Vector2D_ByFloat(Float=2.5, Vector2D=$Particles.SpriteSize)
   ) {isSetParameters=true}
  ]
  rendererClasses [&NiagaraSpriteRendererProperties]
  renderers [
   {rendererClass=&NiagaraSpriteRendererProperties}
   {rendererClass=&NiagaraSpriteRendererProperties, rendererIndex=1}
  ]
 }
 Motes {
  emitterUpdateScript [
   &EmitterState(Life Cycle Mode=::NewEnumerator1, Loop Behavior=::NewEnumerator1, Loop Duration=1, UseLoopDelay=false)
   &SpawnBurst_Instantaneous(
    Spawn Count=&RandomRangeInt(Minimum=11, Maximum=13)
    Spawn Time=0.005
    Use Spawn Probability=false
    UseLoopCountLimit=false
   )
   &SpawnBurst_Instantaneous(Spawn Count=5, Spawn Group=1, Spawn Time=0.035, Use Spawn Probability=false, UseLoopCountLimit=false) {module=SpawnBurst_Instantaneous001}
  ]
  particleSpawnScript [
   /SetVariables(Particles.VisibilityTag=&SpawnGroup()) {isSetParameters=true}
   &InitializeParticle(
    Color=&MaskLinearColor(
     Mask Source=0
     Mask Parameter=?
     Mask=0
     Masked Linear Color=&SelectLinearColorFromArray(
      Array Sampling Mode=::NewEnumerator0
      Direct Array Index=?
      Color Selection Array=DataInterface() {propertyValues {ColorData [
         <rgba 1 0.0377 0.0588 1>
         <rgba 1 0.2039 0.0948 1>
         <rgba 0.3156 1 0.0741 1>
         <rgba 0.075 0.5024 1 1>
         <rgba 0.4604 0.2882 1 1>
         <rgba 1 0.9616 0 1>
        ]}}
      Random Seed=0
     )
     Passthrough Linear Color=&SelectLinearColorFromArray(
      Array Sampling Mode=::NewEnumerator0
      Direct Array Index=?
      Color Selection Array=DataInterface() {propertyValues {ColorData [<rgba 2 0.1784 0.1877 0.3565>, <rgba 2 0.7531 0.1133 0.3565>, <rgba 2 0.2308 1.0021 0.3113>]}}
      Random Seed=0
     )
    )
    Lifetime=&MaskFloat(
     Mask Source=0
     Mask Parameter=?
     Mask=0
     Masked Float=&RandomRangeFloat(Minimum=0.3, Maximum=0.7)
     Passthrough Float=&RandomRangeFloat(Minimum=0.45, Maximum=0.8)
    )
    Sprite Rotation Mode=::NewEnumerator2
    Sprite Size Mode=::NewEnumerator3
    Uniform Sprite Size=&MaskFloat(
     Mask Source=0
     Mask Parameter=?
     Mask=0
     Masked Float=&RandomRangeFloat(Minimum=1.5, Maximum=5)
     Passthrough Float=&RandomRangeFloat(Minimum=12, Maximum=19)
    )
    UsePositionOffset=false
    Write Material Random=true
   )
   &SkeletalMeshLocation(
    MaskSpawnGroup=false
    Mesh Sampling Type=::NewEnumerator5
    Sampled Position Offset=<0 7 3>
    Sampled Position Offset Coordinate Space:ENiagaraSkelMeshTransforms=::NewEnumerator2
    Skeletal Mesh=$System.Skeletal Mesh Reader
    Socket Sampling Mode=::NewEnumerator5
   )
   &ShapeLocation(MaskSpawnGroup=false, Sphere Distribution=::NewEnumerator3, Sphere Radius=40, UseOwnerScale=true)
   &AddVelocity(
    Module.Use Falloff Distance=false
    Origin Offset:NiagaraPosition=$Particles.SkeletalMeshLocation.SampledPosition
    Override Seed=false
    Velocity Mode=::NewEnumerator1
    Velocity Speed=&RandomRangeFloat(Minimum=100, Maximum=350)
   )
  ]
  particleUpdateScript [
   &ParticleState()
   &Drag(Drag=&RandomRangeFloat(Minimum=3, Maximum=7), Use Linear Drag=true, Use Rotational Drag=false)
   &ScaleSpriteSize(Scale Factor=&Vector2DFromFloat(Value=&MaskFloat(
      Mask Source=0
      Mask Parameter=?
      Mask=0
      Masked Float=&FloatFromCurve(FloatCurve=DataInterface() {propertyValues {Curve {keys [
           <at -0.0046 1.0113 RCIM_Cubic>
           {arriveTangent=-0.9195, interpMode=RCIM_Cubic, leaveTangent=-0.9195, tangentMode=RCTM_User, time=0.435, value=0.7826}
           {arriveTangent=-2.1915, tangentMode=RCTM_User, time=1.0011, value=0.002}
          ]}}})
      Passthrough Float=&FloatFromCurve(
       FloatCurve=DataInterface() {propertyValues {Curve {keys [
           {arriveTangent=0.0024, interpMode=RCIM_Cubic, leaveTangent=0.0024, tangentMode=RCTM_User, time=-0.0046, value=1.0113}
           {arriveTangent=-1.3865, interpMode=RCIM_Cubic, leaveTangent=-1.3865, tangentMode=RCTM_User, time=0.4178, value=0.7731}
           {arriveTangent=-0.6173, tangentMode=RCTM_User, time=1.0011, value=0.002}
          ]}}}
       Scale Curve=&RandomRangeFloat(Minimum=0.75, Maximum=1)
      )
     )))
   &ScaleColor(
    Scale Alpha=&MaskFloat(
     Mask Source=0
     Mask Parameter=?
     Mask=0
     Masked Float=&FloatFromCurve(FloatCurve=DataInterface() {propertyValues {Curve {keys [<at -0.0046 1.0113 RCIM_Cubic>,<at 1.0011 0.002 RCIM_Cubic>]}}})
     Passthrough Float=&FloatFromCurve(
      FloatCurve=DataInterface() {propertyValues {Curve {
         defaultValue=1.0113
         keys [
          {arriveTangent=30.0955, interpMode=RCIM_Cubic, leaveTangent=30.0955, tangentMode=RCTM_User, time=0, value=0}
          {arriveTangent=0.4934, interpMode=RCIM_Cubic, leaveTangent=0.4934, tangentMode=RCTM_User, time=0.0756, value=1}
          {arriveTangent=-1.761, interpMode=RCIM_Cubic, leaveTangent=-1.761, tangentMode=RCTM_User, time=0.2335, value=0.5655}
          {arriveTangent=-0.0405, interpMode=RCIM_Cubic, leaveTangent=-0.0405, tangentMode=RCTM_User, time=1, value=0}
         ]
        }}}
      Scale Curve=&RandomRangeFloat(Minimum=0.75, Maximum=1)
     )
    )
    Scale RGB=&MaskVector(
     Mask Source=0
     Mask Parameter=?
     Mask=0
     Masked Vector=<1 1 1>
     Passthrough Vector=&MakeVectorFromLinearColorRGB(Color=&ColorFromCurve(ColorCurve=DataInterface() {propertyValues {
         AlphaCurve {keys []}
         BlueCurve {keys [<at 0 1 RCIM_Cubic>,<at 0.0429 7>,<at 0.1528 1>,<at 1 1 RCIM_Cubic>]}
         GreenCurve {keys [<at 0 1 RCIM_Cubic>,<at 0.0429 7>,<at 0.1528 1>,<at 1 1 RCIM_Cubic>]}
         RedCurve {keys [<at 0 1 RCIM_Cubic>,<at 0.0429 7>,<at 0.1528 1>,<at 1 1 RCIM_Cubic>]}
        }}))
    )
    ScaleA=true
    ScaleRGB=true
   )
   &ScaleVelocity(Velocity Scale=&VectorFromFloat(Value=&FloatFromCurve(FloatCurve=DataInterface() {propertyValues {Curve {
         defaultValue=1
         keys [
          {arriveTangent=221.4643, interpMode=RCIM_Cubic, leaveTangent=221.4643, tangentMode=RCTM_User, time=0, value=0}
          {arriveTangent=-29.9025, interpMode=RCIM_Cubic, leaveTangent=-29.9025, tangentMode=RCTM_User, time=0.0293, value=1.4322}
          {arriveTangent=-1.7641, interpMode=RCIM_Cubic, leaveTangent=-1.7641, tangentMode=RCTM_User, time=0.1016, value=0.5337}
          {arriveTangent=-0.0405, interpMode=RCIM_Cubic, leaveTangent=-0.0405, tangentMode=RCTM_User, time=1, value=0}
         ]
        }}})))
   &SolveForcesAndVelocity(Clamp Velocity=false, Limit Acceleration=false)
  ]
  rendererClasses [&NiagaraSpriteRendererProperties]
  renderers [
   {rendererClass=&NiagaraSpriteRendererProperties}
   {rendererClass=&NiagaraSpriteRendererProperties, rendererIndex=1}
  ]
 }
 New_FlashElements {
  emitterUpdateScript [
   &EmitterState(Life Cycle Mode=::NewEnumerator1, Loop Behavior=::NewEnumerator1, Loop Duration=0.5, UseLoopDelay=false)
   &SpawnRate(
    SpawnRate=&Multiply_Float(
     A=150
     B=&FloatFromCurve(FloatCurve=DataInterface() {propertyValues {Curve {keys [<at 0.005 0>,<at 0.01 1>,<at 0.215 0>]}}}, CurveIndex=$Emitter.NormalizedLoopAge)
    )
    Use Spawn Probability=false
   )
   &SpawnBurst_Instantaneous(Spawn Count=11, Spawn Time=0.005, Use Spawn Probability=false, UseLoopCountLimit=false)
  ]
  particleSpawnScript [
   &InitializeParticle(
    Color=&SelectLinearColorFromArray(
     Array Sampling Mode=::NewEnumerator0
     Direct Array Index=?
     Color Selection Array=DataInterface() {propertyValues {ColorData [<rgba 8 5.1991 0 1>, <rgba 8 0.4229 0 1>, <rgba 2.2136 0 8 1>, <rgba 0 4.0136 8 1>, <rgba 0.0196 8 1.9206 1>]}}
     Random Seed=0
    )
    Lifetime=&RandomRangeFloat(Minimum=0.15, Maximum=0.3)
    Sprite Size Max=<4 6>
    Sprite Size Min=<1 1>
    Sprite Size Mode=::NewEnumerator2
    UsePositionOffset=false
    Write Material Random=true
   )
   &SkeletalMeshLocation(
    MaskSpawnGroup=false
    Mesh Sampling Type=::NewEnumerator5
    Sampled Position Offset=<0 7 3>
    Sampled Position Offset Coordinate Space:ENiagaraSkelMeshTransforms=::NewEnumerator2
    Skeletal Mesh=$System.Skeletal Mesh Reader
    Socket Sampling Mode=::NewEnumerator5
   )
   &ShapeLocation(MaskSpawnGroup=false, Override Seed=false, Sphere Radius=35, UseOwnerScale=true)
   &AddVelocity(
    Module.Use Falloff Distance=false
    Origin Offset:NiagaraPosition=$Particles.SkeletalMeshLocation.SampledPosition
    Override Seed=false
    Velocity Mode=::NewEnumerator1
    Velocity Speed=&Lerp_Float(A=200, B=600, Alpha=&RandomRangeFloat(Minimum=0, Maximum=1))
   )
   &ApplyInitialForces()
   &CameraOffset(Camera Offset Amount=14)
  ]
  particleUpdateScript [
   &ParticleState()
   &Drag(Drag=&RandomRangeFloat(Minimum=10, Maximum=30), Use Linear Drag=true, Use Rotational Drag=true)
   &SolveForcesAndVelocity(Clamp Velocity=false, Limit Acceleration=false)
   &ScaleSpriteSize(Uniform Curve Sprite Scale=DataInterface() {propertyValues {Curve {
       defaultValue=0
       keys [
        {arriveTangent=7.0298, interpMode=RCIM_Cubic, leaveTangent=7.0298, tangentMode=RCTM_User, time=0, value=0.1}
        {arriveTangent=-0.1, arriveTangentWeight=0.06, interpMode=RCIM_Cubic, leaveTangent=-0.1, tangentMode=RCTM_User, time=0.1411, value=1}
        {arriveTangent=-1.4276, interpMode=RCIM_Cubic, leaveTangent=-1.4276, tangentMode=RCTM_User, time=0.2337, value=0.7189}
        {arriveTangent=-0.3031, arriveTangentWeight=0.2733, interpMode=RCIM_Cubic, leaveTangent=-0.3031, tangentMode=RCTM_User, time=1, value=0}
       ]
      }}})
   &ScaleSpriteSizeBySpeed(Max Scale Factor=<1 11>, Min Scale Factor=<0.5 1>)
   &ScaleColor(
    Scale Alpha=&FloatFromCurve(FloatCurve=DataInterface() {propertyValues {Curve {
        defaultValue=1
        keys [
         {interpMode=RCIM_Cubic, tangentMode=RCTM_User, time=0, value=0}
         {interpMode=RCIM_Cubic, tangentMode=RCTM_User, time=0.0614, value=1}
         {arriveTangent=-2.9648, interpMode=RCIM_Cubic, leaveTangent=-2.9648, tangentMode=RCTM_User, time=0.1337, value=0.9484}
         {tangentMode=RCTM_User, time=1, value=0}
        ]
       }}})
    ScaleA=true
    ScaleRGB=true
   )
  ]
  rendererClasses [&NiagaraSpriteRendererProperties]
  renderers [{rendererClass=&NiagaraSpriteRendererProperties}]
 }
 New_FlashElements001 {
  emitterUpdateScript [
   &EmitterState(Life Cycle Mode=::NewEnumerator1, Loop Behavior=::NewEnumerator1, Loop Duration=0.5, UseLoopDelay=false)
   &SpawnBurst_Instantaneous(Spawn Count=9, Spawn Time=0.005, Use Spawn Probability=false, UseLoopCountLimit=false)
  ]
  particleSpawnScript [
   &InitializeParticle(
    Color=&SelectLinearColorFromArray(
     Array Sampling Mode=::NewEnumerator0
     Direct Array Index=?
     Color Selection Array=DataInterface() {propertyValues {ColorData [<rgba 5 3.2495 0 1>, <rgba 5 0.2643 0 1>, <rgba 1.3835 0 5 1>, <rgba 0 2.5085 5 1>, <rgba 0.0122 5 1.2004 1>]}}
     Random Seed=0
    )
    Lifetime=&RandomRangeFloat(Minimum=0.15, Maximum=0.25)
    Sprite Size Max=<13 35>
    Sprite Size Min=<7 25>
    Sprite Size Mode=::NewEnumerator2
    UsePositionOffset=false
    Write Material Random=true
   )
   &SkeletalMeshLocation(
    MaskSpawnGroup=false
    Mesh Sampling Type=::NewEnumerator5
    Sampled Position Offset=<0 7 3>
    Sampled Position Offset Coordinate Space:ENiagaraSkelMeshTransforms=::NewEnumerator2
    Skeletal Mesh=$System.Skeletal Mesh Reader
    Socket Sampling Mode=::NewEnumerator5
   )
   &ShapeLocation(MaskSpawnGroup=false, Sphere Distribution=::NewEnumerator3, Sphere Radius=8, UseOwnerScale=true)
   &AddVelocity(
    Module.Use Falloff Distance=false
    Origin Offset:NiagaraPosition=$Particles.SkeletalMeshLocation.SampledPosition
    Override Seed=false
    Velocity Mode=::NewEnumerator1
    Velocity Speed=&Lerp_Float(A=150, B=330, Alpha=&RandomRangeFloat(Minimum=0, Maximum=1))
   )
   &ApplyInitialForces()
   &SubUVAnimation(
    Output Alternate Renderer Binding=true
    "Sprite Renderer "=DataInterface() {propertyValues {SpriteRenderer=&self:New_FlashElements001_0.NiagaraSpriteRendererProperties_1}}
    SubUV Animation Mode:ENiagara_SubUVLookupModeV2=::NewEnumerator2
    UseEndFrame=false
    UseStartFrame=false
   )
   &CameraOffset(Camera Offset Amount=13)
  ]
  particleUpdateScript [
   &ParticleState()
   &ScaleSpriteSize(
    Non-Uniform Curve Sprite Scale=DataInterface() {propertyValues {XCurve {defaultValue=0.778, keys [<at 0 0.778>,<at 1 1>]}, YCurve {defaultValue=0.778, keys [<at 0 0.05>,<at 1 1.5>]}}}
    Scale Sprite Size Mode=::NewEnumerator3
   )
   &SolveForcesAndVelocity(Clamp Velocity=false, Limit Acceleration=false)
   &ScaleColor(
    Scale Alpha=&FloatFromCurve(FloatCurve=DataInterface() {propertyValues {Curve {
        defaultValue=1
        keys [
         {interpMode=RCIM_Cubic, tangentMode=RCTM_User, time=0, value=0}
         {interpMode=RCIM_Cubic, tangentMode=RCTM_User, time=0.0614, value=1}
         {arriveTangent=-2.9648, interpMode=RCIM_Cubic, leaveTangent=-2.9648, tangentMode=RCTM_User, time=0.1337, value=0.9484}
         {tangentMode=RCTM_User, time=1, value=0}
        ]
       }}})
    Scale RGB=&MakeVectorFromLinearColorRGB(Color=&ColorFromCurve(ColorCurve=DataInterface() {propertyValues {
        AlphaCurve {keys []}
        BlueCurve {keys [
          <at 0.0238 2.7351>
          <at 0.0393 5>
          {arriveTangent=-27.9144, interpMode=RCIM_Cubic, leaveTangent=-27.9144, time=0.1788, value=1}
          <at 0.1826 1 RCIM_Cubic>
          <at 0.7655 1>
          <at 0.9223 0.0838>
         ]}
        GreenCurve {keys [
          <at 0.0238 6.6424>
          <at 0.0393 5>
          {arriveTangent=-27.9144, interpMode=RCIM_Cubic, leaveTangent=-27.9144, time=0.1788, value=1}
          <at 0.1826 1 RCIM_Cubic>
          <at 0.7655 1>
          <at 0.9223 0.1853>
         ]}
        RedCurve {keys [
          <at 0.0238 12>
          <at 0.0393 5>
          {arriveTangent=-27.9144, interpMode=RCIM_Cubic, leaveTangent=-27.9144, time=0.1788, value=1}
          <at 0.1826 1 RCIM_Cubic>
          <at 0.7655 1>
          <at 0.9223 1>
         ]}
       }}))
    ScaleA=true
    ScaleRGB=true
   )
   /SetVariables(
    Particles.Glow Color=&MakeLinearColorFromVectorAndFloat(
     "Vector (RGB)"=&MakeVectorFromLinearColorRGB(Color=$Particles.Color)
     "Float (Alpha)"=&FloatFromCurve(
      FloatCurve=DataInterface() {propertyValues {Curve {
         defaultValue=1
         keys [
          {interpMode=RCIM_Cubic, tangentMode=RCTM_User, time=0, value=0}
          {interpMode=RCIM_Cubic, tangentMode=RCTM_User, time=0.0614, value=1}
          {arriveTangent=-2.9648, interpMode=RCIM_Cubic, leaveTangent=-2.9648, tangentMode=RCTM_User, time=0.1337, value=0.9484}
          {tangentMode=RCTM_User, time=1, value=0}
         ]
        }}}
      Scale Curve=0.8
     )
    )
    Particles.Glow Size=&Multiply_Vector2D(A=$Particles.SpriteSize, B=<1 1.7>)
   ) {isSetParameters=true}
  ]
  rendererClasses [&NiagaraSpriteRendererProperties]
  renderers [
   {rendererClass=&NiagaraSpriteRendererProperties}
   {rendererClass=&NiagaraSpriteRendererProperties, rendererIndex=1}
  ]
 }
 Stars_new {
  emitterUpdateScript [
   &EmitterState(Life Cycle Mode=::NewEnumerator1, Loop Behavior=::NewEnumerator1, Loop Duration=1.1, UseLoopDelay=false)
   &SpawnBurst_Instantaneous(Use Spawn Probability=false, UseLoopCountLimit=false)
   &SpawnBurst_Instantaneous(
    Spawn Count=&RandomRangeInt(Minimum=5, Maximum=7)
    Spawn Group=1
    Spawn Time=0.02
    Use Spawn Probability=false
    UseLoopCountLimit=false
   ) {module=SpawnBurst_Instantaneous001}
   &SpawnBurst_Instantaneous(
    Spawn Count=&RandomRangeInt(Minimum=2, Maximum=4)
    Spawn Group=2
    Spawn Time=0.01
    Use Spawn Probability=false
    UseLoopCountLimit=false
   ) {module=SpawnBurst_Instantaneous002}
  ]
  particleSpawnScript [
   /SetVariables(Particles.VisibilityTag=&SpawnGroup()) {isSetParameters=true}
   &InitializeParticle(
    Color Mode=::NewEnumerator0
    Lifetime=&MaskFloat(
     Mask Source=0
     Mask Parameter=?
     Mask=0
     Masked Float=0.4
     Passthrough Float=&MaskFloat(
      Mask Source=0
      Mask Parameter=?
      Mask=1
      Masked Float=&RandomRangeFloat(Minimum=0.2, Maximum=0.75)
      Passthrough Float=0.9
     )
    )
    Sprite Rotation Angle=&MaskFloat(
     Mask Source=0
     Mask Parameter=?
     Mask=0
     Masked Float=-14
     Passthrough Float=&RandomRangeFloat(Minimum=0, Maximum=360)
    )
    Sprite Rotation Mode=::NewEnumerator0
    Sprite Size Mode=::NewEnumerator3
    Uniform Sprite Size=&MaskFloat(
     Mask Source=0
     Mask Parameter=?
     Mask=0
     Masked Float=29
     Passthrough Float=&RandomRangeFloat(Minimum=5, Maximum=8)
    )
    UsePositionOffset=false
    Write Material Random=true
   )
   &SkeletalMeshLocation(
    MaskSpawnGroup=false
    Mesh Sampling Type=::NewEnumerator5
    Sampled Position Offset=<0 7 3>
    Sampled Position Offset Coordinate Space:ENiagaraSkelMeshTransforms=::NewEnumerator2
    Skeletal Mesh=$System.Skeletal Mesh Reader
    Socket Sampling Mode=::NewEnumerator5
   )
   &ShapeLocation(
    MaskSpawnGroup=false
    Rotation Mode=::NewEnumerator5
    Sphere Distribution=::NewEnumerator3
    Sphere Radius=&MaskFloat(Mask Source=0, Mask Parameter=?, Mask=0, Masked Float=0, Passthrough Float=30)
    UseOwnerScale=true
   )
   &AddVelocity(
    Module.Use Falloff Distance=false
    Override Seed=false
    Velocity Mode=::NewEnumerator1
    Velocity Origin=$Particles.SkeletalMeshLocation.SampledPosition
    Velocity Speed=&MaskFloat(
     Mask Source=0
     Mask Parameter=?
     Mask=0
     Masked Float=0
     Passthrough Float=&RandomRangeFloat(Minimum=75, Maximum=125)
    )
   )
   &ApplyInitialForces()
   &CameraOffset(Camera Offset Amount=20)
  ]
  particleUpdateScript [
   &ParticleState()
   &SpriteRotationRate(Rotation Rate=&MaskFloat(
     Mask Source=0
     Mask Parameter=?
     Mask=0
     Masked Float=0
     Passthrough Float=&MaskFloat(
      Mask Source=0
      Mask Parameter=?
      Mask=1
      Masked Float=&UniformAOrBFloat(
       A=&RandomRangeFloat(Minimum=90, Maximum=120)
       B=&RandomRangeFloat(Minimum=-90, Maximum=-120)
       Evaluation Type=::NewEnumerator0
       Recalculate Random Each Loop=true
       Override Seed=false
      )
      Passthrough Float=&UniformAOrBFloat(
       A=&RandomRangeFloat(Minimum=90, Maximum=120)
       B=&RandomRangeFloat(Minimum=-90, Maximum=-120)
       Evaluation Type=::NewEnumerator0
       Recalculate Random Each Loop=true
       Override Seed=false
      )
     )
    ))
   &ScaleSpriteSize(
    Scale Sprite Size Mode=::NewEnumerator0
    Uniform Scale Factor=&MaskFloat(
     Mask Source=0
     Mask Parameter=?
     Mask=0
     Masked Float=0
     Passthrough Float=&MaskFloat(
      Mask Source=0
      Mask Parameter=?
      Mask=1
      Masked Float=&FloatFromCurve(
       FloatCurve=DataInterface() {propertyValues {Curve {
          defaultValue=0
          keys [
           {arriveTangent=3.7898, interpMode=RCIM_Cubic, leaveTangent=3.7898, tangentMode=RCTM_User, time=0, value=0.5}
           {
            arriveTangent=0.2359
            arriveTangentWeight=0.0304
            interpMode=RCIM_Cubic
            leaveTangent=0.2359
            leaveTangentWeight=0.303
            tangentMode=RCTM_User
            time=0.2178
            value=2.1325
           }
           {arriveTangent=-5.1829, interpMode=RCIM_Cubic, leaveTangent=-5.1829, time=0.2422, value=1.2848}
           {arriveTangent=-1.6955, interpMode=RCIM_Cubic, leaveTangent=-1.6955, time=0.5409, value=0.4575}
           {arriveTangent=-0.1129, arriveTangentWeight=0.3333, interpMode=RCIM_Cubic, leaveTangent=-0.1129, tangentMode=RCTM_User, time=1, value=0}
          ]
         }}}
       Scale Curve=1.25
      )
      Passthrough Float=&FloatFromCurve(
       FloatCurve=DataInterface() {propertyValues {Curve {
          defaultValue=0
          keys [
           {arriveTangent=3.7898, interpMode=RCIM_Cubic, leaveTangent=3.7898, tangentMode=RCTM_User, time=0, value=0.5}
           {
            arriveTangent=0.2359
            arriveTangentWeight=0.0304
            interpMode=RCIM_Cubic
            leaveTangent=0.2359
            leaveTangentWeight=0.303
            tangentMode=RCTM_User
            time=0.2178
            value=2.1325
           }
           {arriveTangent=-1.306, interpMode=RCIM_Cubic, leaveTangent=-1.306, tangentMode=RCTM_User, time=0.2422, value=1.2848}
           {arriveTangent=-1.6955, interpMode=RCIM_Cubic, leaveTangent=-1.6955, time=0.607, value=0.7549}
           {arriveTangent=-1.8844, arriveTangentWeight=0.3333, interpMode=RCIM_Cubic, leaveTangent=-1.8844, tangentMode=RCTM_User, time=1, value=0}
          ]
         }}}
       Scale Curve=1.25
      )
     )
    )
   ) {module=ScaleSpriteSize001}
   &Drag(
    Drag=&MaskFloat(
     Mask Source=0
     Mask Parameter=?
     Mask=0
     Masked Float=0
     Passthrough Float=&FloatFromCurve(
      FloatCurve=DataInterface() {propertyValues {Curve {
         defaultValue=1
         keys [
          {interpMode=RCIM_Cubic, tangentMode=RCTM_User, time=0.2041, value=0}
          {arriveTangent=5.7919, interpMode=RCIM_Cubic, leaveTangent=5.7919, tangentMode=RCTM_User, time=0.4986, value=1}
         ]
        }}}
      Scale Curve=&RandomRangeFloat(Minimum=7, Maximum=20)
     )
    )
    Use Linear Drag=true
    Use Rotational Drag=true
   )
   &SolveForcesAndVelocity(Clamp Velocity=false, Limit Acceleration=false)
   &ScaleColor(
    Linear Color Curve=DataInterface() {propertyValues {
      AlphaCurve {keys [<at 0 1>]}
      BlueCurve {keys [<at 0 4.2446>,<at 0.113 1.4856>,<at 0.3461 0.2122>]}
      GreenCurve {keys [<at 0 9.8204>,<at 0.113 3.4371>,<at 0.3461 0.491>]}
      RedCurve {keys [<at 0 20>,<at 0.113 7>,<at 0.3461 1>]}
     }}
    Scale Mode=2
   )
   /SetVariables(Particles.Big Star Size:Vector2f=&Vector2DFromFloat(Value=&Multiply_Float(
      A=30
      B=&FloatFromCurve(FloatCurve=DataInterface() {propertyValues {Curve {
          defaultValue=0.5
          keys [
           {tangentMode=RCTM_User, time=0, value=0.5}
           <at 0.0553 2>
           <at 0.2595 2>
           <at 0.2639 1.0741>
           <at 0.3863 0.6382>
           {tangentMode=RCTM_User, time=0.8622, value=0.2255}
           <at 1 0>
          ]
         }}})
     ))) {isSetParameters=true}
  ]
  rendererClasses [&NiagaraSpriteRendererProperties]
  renderers [
   {rendererClass=&NiagaraSpriteRendererProperties}
   {rendererClass=&NiagaraSpriteRendererProperties, rendererIndex=1}
   {rendererClass=&NiagaraSpriteRendererProperties, rendererIndex=2}
  ]
 }
}
summary {
 systemName=NS_Emote_ProngCling_Stars
 emitters [
  {emitterName=Flare_With_Streak, rendererClasses [&NiagaraSpriteRendererProperties]}
  {emitterName=Motes, rendererClasses [&NiagaraSpriteRendererProperties]}
  {emitterName=Stars_new, rendererClasses [&NiagaraSpriteRendererProperties]}
  {emitterName=New_FlashElements, rendererClasses [&NiagaraSpriteRendererProperties]}
  {emitterName=New_FlashElements001, rendererClasses [&NiagaraSpriteRendererProperties]}
  {emitterName=Fountain, rendererClasses [&NiagaraSpriteRendererProperties]}
 ]
 userVariables [
  User.bIsFrontEnd {type=NiagaraBool}
  User.bIsFrontEndPreview {type=NiagaraBool}
 ]
}
